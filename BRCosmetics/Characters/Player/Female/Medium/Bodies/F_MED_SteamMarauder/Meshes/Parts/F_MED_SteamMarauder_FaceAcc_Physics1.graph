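# Physics asset for the female medium SteamMarauder face accessory, covering its dreadlock bones with sphere and capsule bodies.
# Limited swing/twist constraints link front, right-side and branching right-ear chains, clavicle and spine dread pairs, and sheafBase to deform_megaLookUp.
# Front chains attach to the head, whose neck joint locks swing and twist.
# Head, neck, upper spine, clavicle, arm and right weapon bodies supply surrounding body shapes.
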PhysicsAsset /BRCosmetics/Characters/Player/Female/Medium/Bodies/F_MED_SteamMarauder/Meshes/Parts/F_MED_SteamMarauder_FaceAcc_Physics1
bodies {
 FaceDreadlockFront_l_1 [{shapeType=Sphere, radius=1.414, shapeName=FaceDreadlockFront_l_1_sphere}]
 FaceDreadlockFront_l_2 [{shapeType=Capsule, length=3.4304, radius=1.9108, shapeName=FaceDreadlockFront_l_2_capsule, center=<1.9967 0 -0.0328>, rotation=<rot 90 0 0>}]
 FaceDreadlockFront_l_3 [{
   shapeType=Capsule
   length=3.5591
   radius=1.6108
   shapeName=FaceDreadlockFront_l_3_capsule
   center=<1.952 0 0>
   rotation=<rot 0 0.0005 -74.2106>
  }]
 FaceDreadlockFront_l_4 [
  {
   shapeType=Capsule
   length=3.7045
   radius=1.4782
   shapeName=FaceDreadlockFront_l_2_capsule
   center=<0.9163 -0.0319 0.1642>
   rotation=<rot 60.7174 102.5917 10.2362>
  }
  {
   shapeType=Capsule
   length=2.1155
   radius=1.5028
   shapeName=FaceDreadlockFront_l_4_capsule
   center=<4.1957 0.9779 0.0032>
   rotation=<rot 0 0.0001 -45.4387>
  }
 ]
 FaceDreadlockFront_l_5 [
  {shapeType=Sphere, radius=1.6453, shapeName=FaceDreadlockFront_l_5_sphere, center=<2.683 1.6925 -2.0488>}
  {
   shapeType=Capsule
   length=4.3524
   radius=1.1282
   shapeName=FaceDreadlockFront_l_2_capsule
   center=<3.566 0.399 0.2308>
   rotation=<rot 90 -179.8881 179.7328>
  }
 ]
 FaceDreadlockFront_r_1 [{shapeType=Capsule, length=3, radius=1.5, shapeName=FaceDreadlockFront_r_1_capsule, center=<2.389 0 0>, rotation=<rot 90 0 0>}]
 FaceDreadlockFront_r_2 [{shapeType=Capsule, length=3, radius=1.5, shapeName=FaceDreadlockFront_r_1_capsule, center=<2.389 0 0>, rotation=<rot 90 0 0>}]
 FaceDreadlockFront_r_3 [{shapeType=Capsule, length=3, radius=1.5, shapeName=FaceDreadlockFront_r_1_capsule, center=<2.389 0 0>, rotation=<rot 90 0 0>}]
 FaceDreadlockFront_r_4 [{shapeType=Capsule, length=2.4685, radius=2.0315, shapeName=FaceDreadlockFront_r_1_capsule, center=<2.389 0 0>, rotation=<rot 90 0 0>}]
 FaceDreadlockFront_r_5 [{shapeType=Capsule, length=10.2221, radius=1.934, shapeName=FaceDreadlockFront_r_1_capsule, center=<7.3349 0 0>, rotation=<rot 90 0 0>}]
 FaceDreadlockSide_r_1 [{shapeType=Capsule, length=3.5, radius=1.5, shapeName=FaceDreadlockFront_r_1_capsule, center=<3.0725 0 0>, rotation=<rot 90 0 0>}]
 FaceDreadlockSide_r_2 [{shapeType=Capsule, length=3.5, radius=1.5, shapeName=FaceDreadlockFront_r_1_capsule, center=<3.0725 0 0>, rotation=<rot 90 0 0>}]
 FaceDreadlockSide_r_3 [{shapeType=Capsule, length=4.5, radius=1.5, shapeName=FaceDreadlockFront_r_1_capsule, center=<2.7516 0 0>, rotation=<rot 90 0 0>}]
 FaceDreadlockSide_r_4 [{shapeType=Capsule, length=5.1429, radius=1.5, shapeName=FaceDreadlockFront_r_1_capsule, center=<2.9938 -0.0716 0>, rotation=<rot 90 0 0>}]
 FaceDreadlockSide_r_5 [{shapeType=Capsule, length=3, radius=1.5, shapeName=FaceDreadlockFront_r_1_capsule, center=<1.4589 0 0>, rotation=<rot 90 0 0>}]
 clavicle_l [{
   shapeType=Capsule
   length=8.9299
   radius=4.8318
   shapeName=clavicle_l_capsule
   center=<5.1163 -1.2253 -0.3006>
   rotation=<rot -80 179.9982 -179.9993>
  }]
 clavicle_r [{
   shapeType=Capsule
   length=6.3701
   radius=4.7983
   shapeName=clavicle_l_capsule
   center=<-5.1163 1.1903 0.3006>
   rotation=<rot -80 179.9982 0.0007>
  }]
 deform_megaLookUp [{shapeType=Sphere, radius=3.1109, shapeName=deform_megaLookUp_sphere}]
 dread_clavicle_l [{shapeType=Sphere, radius=3.9663, shapeName=dread_clavicle_l_sphere, center=<0.479 0.3863 3.0075>}]
 dread_clavicle_r [{shapeType=Sphere, radius=3.4599, shapeName=dread_clavicle_r_sphere}]
 dread_spine_l [{shapeType=Sphere, radius=1.8373, shapeName=dread_clavicle_l_sphere}]
 dread_spine_r [{shapeType=Sphere, radius=1.8373, shapeName=dread_clavicle_l_sphere}]
 dyn_dread_clavicle_l [{
   shapeType=Capsule
   length=3.3236
   radius=1.4193
   shapeName=dyn_dread_clavicle_l_capsule
   center=<3.3817 0.5368 -0.7408>
   rotation=<rot -83.6425 179.9999 -179.9999>
  }]
 dyn_dread_clavicle_r [{
   shapeType=Capsule
   length=4.1762
   radius=2.1407
   shapeName=dyn_dread_clavicle_l_capsule
   center=<2.6408 0 0>
   rotation=<rot -90 179.367 -179.367>
  }]
 dyn_dread_spine_l [{
   shapeType=Capsule
   length=5.4247
   radius=1.6114
   shapeName=dyn_dread_clavicle_l_capsule
   center=<3.3703 0 -0.7856>
   rotation=<rot -79.9999 180 -180>
  }]
 dyn_dread_spine_r_A [{
   shapeType=Capsule
   length=4.3131
   radius=2.0038
   shapeName=dyn_dread_clavicle_l_capsule
   center=<2.6408 0 0>
   rotation=<rot -90 179.367 -179.367>
  }]
 dyn_dread_spine_r_B [{
   shapeType=Capsule
   length=4.1365
   radius=2.1804
   shapeName=dyn_dread_clavicle_l_capsule
   center=<2.5308 0 -0.6241>
   rotation=<rot -80 -179.9999 179.9999>
  }]
 earDreadlock_r_1 [{shapeType=Sphere, radius=2.1796, shapeName=earDreadlock_r_1_sphere}]
 earDreadlock_r_2 [{shapeType=Capsule, length=7.1127, radius=2, shapeName=earDreadlock_r_1_capsule, center=<3.5106 0.0689 0.2002>, rotation=<rot 90 0 -11.6984>}]
 earDreadlock_r_A1 [{shapeType=Capsule, length=4.5298, radius=1.1977, shapeName=FaceDreadlockFront_r_1_capsule, center=<4.9628 0.3397 0>, rotation=<rot 90 0 0>}]
 earDreadlock_r_B1 [{
   shapeType=Capsule
   length=3.0215
   radius=1.227
   shapeName=FaceDreadlockFront_r_1_capsule
   center=<5.097 0.5359 -2.0982>
   rotation=<rot 64.2772 0.0002 0.0001>
  }]
 head [
  {shapeType=Capsule, length=5.2995, radius=6.3052, shapeName=head_capsule, center=<2.2699 3.7467 0>, rotation=<rot 0 89.9991 -90>}
  {shapeType=Capsule, length=4.0741, radius=7.9673, shapeName=head_capsule, center=<5.851 -1.2645 0>, rotation=<rot 90 180 80.8412>}
 ]
 lowerarm_r [{shapeType=Capsule, length=25.7771, radius=4.223, shapeName=lowerarm_r_capsule, center=<-10.6148 0 0>, rotation=<rot -90 179.9802 -179.9802>}]
 neck_01 [{shapeType=Capsule, length=7.0202, radius=3.9798, shapeName=neck_01_capsule, center=<3.3336 0 0>, rotation=<rot 90 179.9999 -179.9997>}]
 sheafBase [{shapeType=Sphere, radius=6.5185, shapeName=sheafBase_sphere, center=<4.1445 0 0>}]
 spine_04 [
  {shapeType=Capsule, length=8.2256, radius=7.9468, shapeName=spine_04_capsule, center=<5.4836 -0.2931 0>}
  {shapeType=Capsule, length=6.9067, radius=7.4607, shapeName=spine_04_capsule, center=<6.1931 1.6282 0>}
  {shapeType=Capsule, length=9.0852, radius=6.427, shapeName=spine_04_capsule, center=<0.7595 6.1828 0>}
 ]
 spine_05 [
  {shapeType=Capsule, length=3.214, radius=2.9493, shapeName=spine_05_capsule, center=<5.2038 5.6619 0>, rotation=<rot 0 35.8876 0>}
  {shapeType=Capsule, length=8.0846, radius=7.6142, shapeName=spine_05_capsule, center=<1.0453 0.0342 0>}
 ]
 upperarm_l [{
   shapeType=Capsule
   length=19.6191
   radius=4.8809
   shapeName=upperarm_l_capsule
   center=<11.7233 -0.7553 -0.7605>
   rotation=<rot -90 179.7762 177.2607>
  }]
 upperarm_r [{
   shapeType=Capsule
   length=21.2749
   radius=3.2251
   shapeName=upperarm_l_capsule
   center=<-10.2497 1.1024 0.3099>
   rotation=<rot -90 166.7299 13.3815>
  }]
 weapon_r [{
   shapeType=Capsule
   length=41
   radius=3.5
   shapeName=ik_hand_gun_capsule
   center=<2.9071 -3.1707 6.3678>
   rotation=<rot 87.2812 -0.0012 -0.0013>
  }]
}
constraints [
 {bone1Name=FaceDreadlockSide_r_3, bone2Name=FaceDreadlockSide_r_2, swing1LimitDegrees=20, swing2LimitDegrees=20, twistLimitDegrees=2}
 {bone1Name=FaceDreadlockSide_r_4, bone2Name=FaceDreadlockSide_r_3, swing1LimitDegrees=20, swing2LimitDegrees=20, twistLimitDegrees=2}
 {bone1Name=FaceDreadlockSide_r_5, bone2Name=FaceDreadlockSide_r_4, swing1LimitDegrees=20, swing2LimitDegrees=20, twistLimitDegrees=2}
 {bone1Name=dyn_dread_clavicle_r, bone2Name=dread_clavicle_r, swing1LimitDegrees=15, swing2LimitDegrees=32, twistLimitDegrees=3}
 {bone1Name=dyn_dread_clavicle_l, bone2Name=dread_clavicle_l, swing1LimitDegrees=30, swing2LimitDegrees=30, twistLimitDegrees=3}
 {bone1Name=dyn_dread_spine_l, bone2Name=dread_spine_l, swing1LimitDegrees=30, swing2LimitDegrees=30, twistLimitDegrees=3, twistMotion=Locked}
 {
  bone1Name=dyn_dread_spine_r_A
  bone2Name=dread_spine_r
  swing1LimitDegrees=30
  swing2LimitDegrees=30
  twistLimitDegrees=3
  twistMotion=Locked
 }
 {
  bone1Name=dyn_dread_spine_r_B
  bone2Name=dread_spine_r
  swing1LimitDegrees=30
  swing2LimitDegrees=30
  twistLimitDegrees=3
  twistMotion=Locked
 }
 {
  bone1Name=neck_01
  bone2Name=head
  swing1LimitDegrees=45
  swing1Motion=Locked
  swing2LimitDegrees=45
  swing2Motion=Locked
  twistLimitDegrees=45
  twistMotion=Locked
 }
 {
  bone1Name=earDreadlock_r_2
  bone2Name=earDreadlock_r_1
  swing1LimitDegrees=25
  swing2LimitDegrees=25
  twistLimitDegrees=25
  twistMotion=Locked
 }
 {bone1Name=FaceDreadlockFront_l_5, bone2Name=FaceDreadlockFront_l_4, swing1LimitDegrees=20, swing2LimitDegrees=20, twistLimitDegrees=15}
 {bone1Name=FaceDreadlockFront_l_4, bone2Name=FaceDreadlockFront_l_3, swing1LimitDegrees=30, swing2LimitDegrees=30, twistLimitDegrees=15}
 {bone1Name=FaceDreadlockFront_l_3, bone2Name=FaceDreadlockFront_l_2, swing1LimitDegrees=10, swing2LimitDegrees=10, twistLimitDegrees=2}
 {bone1Name=FaceDreadlockFront_l_2, bone2Name=FaceDreadlockFront_l_1, swing1LimitDegrees=30, swing2LimitDegrees=20, twistLimitDegrees=2}
 {bone1Name=FaceDreadlockFront_r_5, bone2Name=FaceDreadlockFront_r_4, swing1LimitDegrees=25, swing2LimitDegrees=25, twistLimitDegrees=3}
 {bone1Name=FaceDreadlockFront_r_4, bone2Name=FaceDreadlockFront_r_3, swing1LimitDegrees=25, swing2LimitDegrees=25, twistLimitDegrees=3}
 {bone1Name=FaceDreadlockFront_r_3, bone2Name=FaceDreadlockFront_r_2, swing1LimitDegrees=25, swing2LimitDegrees=25, twistLimitDegrees=3}
 {bone1Name=FaceDreadlockFront_r_2, bone2Name=FaceDreadlockFront_r_1, swing1LimitDegrees=25, swing2LimitDegrees=25, twistLimitDegrees=3}
 {bone1Name=FaceDreadlockFront_r_1, bone2Name=head, swing1LimitDegrees=45, swing2LimitDegrees=45, twistLimitDegrees=45}
 {bone1Name=earDreadlock_r_B1, bone2Name=earDreadlock_r_2, swing1LimitDegrees=35, swing2LimitDegrees=35, twistLimitDegrees=20}
 {bone1Name=earDreadlock_r_A1, bone2Name=earDreadlock_r_2, swing1LimitDegrees=25, swing2LimitDegrees=25, twistLimitDegrees=10}
 {bone1Name=FaceDreadlockSide_r_2, bone2Name=FaceDreadlockSide_r_1, swing1LimitDegrees=20, swing2LimitDegrees=20, twistLimitDegrees=2}
 {bone1Name=sheafBase, bone2Name=deform_megaLookUp, swing1LimitDegrees=20, swing2LimitDegrees=20, twistLimitDegrees=2}
 {bone1Name=FaceDreadlockFront_l_1, bone2Name=head, swing1LimitDegrees=30, swing2LimitDegrees=20, twistLimitDegrees=10}
]
