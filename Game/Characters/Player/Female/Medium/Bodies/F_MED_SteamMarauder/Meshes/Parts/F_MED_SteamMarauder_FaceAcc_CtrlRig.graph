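# Control Rig for the female Steam Marauder face accessory that applies corrective poses to dreadlock bones.
# A sequence first derives head and clavicle weights from spherical pose readers and remapped Euler angles.
# Those weights drive layered transform offsets and constraints, then Kalman-smoothed Targeting blends an aiming correction.
# RB Enable switches on smoothed jump and fall poses, suppressed during SkyDive, and switches off the face-dread offsets.
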ControlRigBlueprint /Game/Characters/Player/Female/Medium/Bodies/F_MED_SteamMarauder/Meshes/Parts/F_MED_SteamMarauder_FaceAcc_CtrlRig
graphs [
 Multiply_1_1_ContainedGraph {nodes [
   n1 = RigVMFunctionEntryNode()
   RigVMFunctionReturnNode(Result=&RigVMFunction_MathDoubleMul(B=@n1.C, A=&RigVMFunction_MathDoubleMul(A=@n1.A, B=@n1.B).Result).Result)
  ]}
 RigVMFunctionLibrary {links [], nodes []}
 RigVMFunction_Sequence_1_ContainedGraph {nodes [
   n1 = RigVMFunctionEntryNode()
   RigVMFunction_Sequence = &RigVMFunction_Sequence(ExecuteContext=@n1.ExecuteContext)
   n2 = &RigVMFunction_Sequence(ExecuteContext=@RigVMFunction_Sequence.B)
   n3 = &RigVMFunction_Sequence(ExecuteContext=@n2.B)
   RigVMFunctionReturnNode(A=@RigVMFunction_Sequence.A, B=@n2.A, C=@n3.A, D=@n3.B)
  ]}
 RigVMModel {nodes [
   RigUnit_BeginExecution = &RigUnit_BeginExecution()
   RigVMFunction_Sequence = RigVMAggregateNode(ExecuteContext=@RigUnit_BeginExecution.ExecutePin)
   RigUnit_GetRelativeTransformForItem = &RigUnit_GetRelativeTransformForItem(Child=(Type=Bone,Name="head"), Parent=(Type=Bone,Name="spine_05"), bChildInitial=False, bParentInitial=False)
   n1 = &RigVMFunction_MathQuaternionToEuler(RotationOrder=ZYX, Value=@RigUnit_GetRelativeTransformForItem.RelativeTransform.Rotation)
   RigVMVariableNode(
    Value=&RigVMFunction_MathDoubleRemap(SourceMaximum=-60, SourceMinimum=-24, TargetMaximum=1, TargetMinimum=0, bClamp=False, Value=@n1.Result.X).Result
    ExecuteContext=@RigVMFunction_Sequence.A
   ) {variable=Head Left}
   > RigVMVariableNode(Value=&RigVMFunction_MathDoubleRemap(SourceMaximum=50, SourceMinimum=4, TargetMaximum=1, TargetMinimum=0, bClamp=False, Value=@n1.Result.X).Result) {variable=Head Right}
   > n2 = &RigUnit_SphericalPoseReader(
    ActiveRegionScaleFactors=(PositiveWidth=1,NegativeWidth=1,PositiveHeight=1,NegativeHeight=1)
    ActiveRegionSize=0.05
    Debug=(bDrawDebug=false,bDraw2D=False,bDrawLocalAxes=true,DebugScale=25,DebugSegments=20,DebugThickness=0.25)
    DriverAxis=(X=0,Y=1,Z=0)
    DriverItem=(Type=Bone,Name="head")
    FalloffRegionScaleFactors=(PositiveWidth=1,NegativeWidth=1,PositiveHeight=1,NegativeHeight=1)
    FalloffSize=0.5
    FlipHeightScaling=False
    FlipWidthScaling=False
    OptionalParentItem=(Type=Bone,Name="spine_05")
    RotationOffset=(X=90,Y=0,Z=-90)
   )
   > RigVMVariableNode(Value=@n2.OutputParam) {variable=Head Up}
   > n3 = &RigUnit_SphericalPoseReader(
    ActiveRegionScaleFactors=(PositiveWidth=1,NegativeWidth=1,PositiveHeight=1,NegativeHeight=1)
    ActiveRegionSize=0.1
    Debug=(bDrawDebug=false,bDraw2D=False,bDrawLocalAxes=true,DebugScale=25,DebugSegments=20,DebugThickness=0.25)
    DriverAxis=(X=0,Y=1,Z=0)
    DriverItem=(Type=Bone,Name="head")
    FalloffRegionScaleFactors=(PositiveWidth=1,NegativeWidth=1,PositiveHeight=1,NegativeHeight=1)
    FalloffSize=0.32
    FlipHeightScaling=False
    FlipWidthScaling=False
    OptionalParentItem=(Type=Bone,Name="spine_05")
    RotationOffset=(X=90,Y=0,Z=45)
   )
   > RigVMVariableNode(Value=@n3.OutputParam) {variable=Head Down}
   > n4 = &RigUnit_SphericalPoseReader(
    ActiveRegionScaleFactors=(PositiveWidth=1,NegativeWidth=1,PositiveHeight=1,NegativeHeight=1)
    ActiveRegionSize=0.2
    Debug=(bDrawDebug=false,bDraw2D=False,bDrawLocalAxes=true,DebugScale=25,DebugSegments=20,DebugThickness=0.25)
    DriverAxis=(X=1,Y=0,Z=0)
    DriverItem=(Type=Bone,Name="head")
    FalloffRegionScaleFactors=(PositiveWidth=1,NegativeWidth=1,PositiveHeight=1,NegativeHeight=1)
    FalloffSize=0.5
    FlipHeightScaling=False
    FlipWidthScaling=False
    OptionalParentItem=(Type=Bone,Name="spine_05")
    RotationOffset=(X=0,Y=0,Z=0)
   )
   > RigVMVariableNode(Value=@n4.OutputParam) {variable=Head roll Right}
   > n5 = &RigUnit_SphericalPoseReader(
    ActiveRegionScaleFactors=(PositiveWidth=1,NegativeWidth=1,PositiveHeight=1,NegativeHeight=1)
    ActiveRegionSize=0.2
    Debug=(bDrawDebug=false,bDraw2D=False,bDrawLocalAxes=true,DebugScale=25,DebugSegments=20,DebugThickness=0.25)
    DriverAxis=(X=1,Y=0,Z=0)
    DriverItem=(Type=Bone,Name="head")
    FalloffRegionScaleFactors=(PositiveWidth=1,NegativeWidth=1,PositiveHeight=1,NegativeHeight=1)
    FalloffSize=0.5
    FlipHeightScaling=False
    FlipWidthScaling=False
    OptionalParentItem=(Type=Bone,Name="spine_05")
    RotationOffset=(X=180,Y=0,Z=0)
   )
   > RigVMVariableNode(Value=@n5.OutputParam) {variable=Head Roll Left}
   > &RigUnit_SphericalPoseReader(
    ActiveRegionScaleFactors=(PositiveWidth=1,NegativeWidth=1,PositiveHeight=1,NegativeHeight=1)
    ActiveRegionSize=0.2
    Debug=(bDrawDebug=false,bDraw2D=False,bDrawLocalAxes=False,DebugScale=25,DebugSegments=20,DebugThickness=0.25)
    DriverAxis=(X=1,Y=0,Z=0)
    DriverItem=(Type=Bone,Name="upperarm_l")
    FalloffRegionScaleFactors=(PositiveWidth=1,NegativeWidth=1,PositiveHeight=1,NegativeHeight=1)
    FalloffSize=0.4
    FlipHeightScaling=False
    FlipWidthScaling=False
    OptionalParentItem=(Type=Bone,Name="spine_05")
    RotationOffset=(X=0,Y=-20,Z=0)
   )
   > &RigUnit_SphericalPoseReader(
    ActiveRegionScaleFactors=(PositiveWidth=1,NegativeWidth=1,PositiveHeight=1,NegativeHeight=1)
    ActiveRegionSize=0.2
    Debug=(bDrawDebug=false,bDraw2D=False,bDrawLocalAxes=False,DebugScale=25,DebugSegments=20,DebugThickness=0.25)
    DriverAxis=(X=-1,Y=0,Z=0)
    DriverItem=(Type=Bone,Name="upperarm_r")
    FalloffRegionScaleFactors=(PositiveWidth=1,NegativeWidth=1,PositiveHeight=1,NegativeHeight=1)
    FalloffSize=0.4
    FlipHeightScaling=False
    FlipWidthScaling=false
    OptionalParentItem=(Type=Bone,Name="spine_05")
    RotationOffset=(X=180,Y=-20,Z=0)
   )
   n6 = &RigVMFunction_MathQuaternionToEuler(
    RotationOrder=XZY
    Value=&RigUnit_GetTransform(Item=(Type=Bone,Name="clavicle_r"), Space=LocalSpace, bInitial=False).Transform.Rotation
   )
   > RigVMVariableNode(Value=&RigVMFunction_MathDoubleRemap(SourceMaximum=127, SourceMinimum=87, TargetMaximum=1, TargetMinimum=0, bClamp=false, Value=@n6.Result.Y).Result) {variable=R clavicle up}
   n7 = &RigUnit_GetTransform(Item=(Type=Bone,Name="clavicle_l"), Space=LocalSpace, bInitial=False)
   > RigVMVariableNode(Value=&RigVMFunction_MathDoubleRemap(
     SourceMaximum=55
     SourceMinimum=93.6
     TargetMaximum=1
     TargetMinimum=0
     bClamp=false
     Value=&RigVMFunction_MathQuaternionToEuler(RotationOrder=XZY, Value=@n7.Transform.Rotation).Result.Y
    ).Result) {variable=L clavicle up}
   Remap = &RigVMFunction_MathDoubleRemap(SourceMaximum=30, SourceMinimum=5, TargetMaximum=1, TargetMinimum=0, bClamp=false, Value=@n6.Result.Z)
   > RigVMVariableNode(Value=@Remap.Result) {variable=R clavicle Back}
   > RigVMVariableNode(Value=&RigVMFunction_MathDoubleRemap(
     SourceMaximum=-22
     SourceMinimum=-5
     TargetMaximum=1
     TargetMinimum=0
     bClamp=false
     Value=&RigVMFunction_MathQuaternionToEuler(RotationOrder=ZXY, Value=@n7.Transform.Rotation).Result.Z
    ).Result) {variable=L clavicle Back}
   &RigUnit_ModifyTransforms(
    ItemToModify="((Item=(Type=Bone,Name=\"deform_megaLookUp\"),Transform=(Rotation=(X=-0.3342,Y=0,Z=0,W=0.9425),Translation=(X=0,Y=3.6624,Z=-2.8911),Scale3D=(X=1,Y=1,Z=1))),(Item=(Type=Bone,Name=\"back_dread_base\"),Transform=(Rotation=(X=-0.0685,Y=-0.0372,Z=0.0597,W=0.9952),T ...(truncated)"
    WeightMinimum=0
    ExecutePin=@RigVMFunction_Sequence.B
    Weight=RigVMVariableNode() {variable=Head Up}.Value
   )
   > &RigUnit_AimConstraintLocalSpaceOffset(
    AdvancedSettings=(DebugSettings=(bEnabled=true,Scale=10,WorldOffset=(Rotation=(X=0,Y=0,Z=0,W=1),Translation=(X=0,Y=0,Z=0),Scale3D=(X=1,Y=1,Z=1))),RotationOrderForFilter=XZY)
    AimAxis=(X=0,Y=0,Z=-1)
    Child=(Type=Bone,Name="back_dread_base")
    Filter=(bX=true,bY=true,bZ=true)
    Parents=((Item=(Type=Bone,Name="spine_05"),Weight=1))
    UpAxis=(X=1,Y=0,Z=0)
    Weight=1
    WorldUp=(Target=(X=0,Y=1,Z=0),Kind=Direction,Space=(Type=Bone,Name="spine_05"))
    bMaintainOffset=true
   )
   > &RigUnit_ParentConstraint(
    AdvancedSettings=(InterpolationType=Average,RotationOrderForFilter=XZY)
    Child=(Type=Bone,Name="dread_clavicle_l")
    Parents=((Item=(Type=Bone,Name="spine_05"),Weight=1))
    Weight=1
   )
   > &RigUnit_ParentConstraint(
    AdvancedSettings=(InterpolationType=Average,RotationOrderForFilter=XZY)
    Child=(Type=Bone,Name="dread_spine_l")
    Parents=((Item=(Type=Bone,Name="spine_05"),Weight=1))
    Weight=1
   )
   > &RigUnit_ParentConstraint(
    AdvancedSettings=(InterpolationType=Average,RotationOrderForFilter=XZY)
    Child=(Type=Bone,Name="dread_clavicle_r")
    Parents=((Item=(Type=Bone,Name="spine_05"),Weight=1))
    Weight=1
    bMaintainOffset=true
   )
   > &RigUnit_ParentConstraint(
    AdvancedSettings=(InterpolationType=Average,RotationOrderForFilter=XZY)
    Child=(Type=Bone,Name="dread_spine_r")
    Parents=((Item=(Type=Bone,Name="spine_05"),Weight=1))
    Weight=1
   )
   > &RigUnit_ModifyTransforms(
    ItemToModify="((Item=(Type=Bone,Name=\"dread_clavicle_l\"),Transform=(Rotation=(X=0,Y=-0.2588,Z=0,W=0.9659),Translation=(X=-2,Y=2,Z=-1),Scale3D=(X=1,Y=1,Z=1))),(Item=(Type=Bone,Name=\"dread_spine_l\"),Transform=(Rotation=(X=0,Y=0,Z=0,W=1),Translation=(X=0,Y=2,Z=3),Scale3D=( ...(truncated)"
    WeightMinimum=0
    Weight=RigVMVariableNode() {variable=Head Right}.Value
   )
   n8 = RigVMVariableNode() {variable=L clavicle Back}
   > &RigUnit_ModifyTransforms(
    ItemToModify=((Item=(Type=Bone,Name="dread_clavicle_l"),Transform=(Rotation=(X=0,Y=-0.0436,Z=0,W=0.999),Translation=(X=0,Y=0,Z=1),Scale3D=(X=1,Y=1,Z=1))))
    WeightMinimum=0
    Weight=@n8.Value
   )
   > &RigUnit_ModifyTransforms(
    ItemToModify=((Item=(Type=Bone,Name="dread_clavicle_l"),Transform=(Rotation=(X=-0.0872,Y=0,Z=0,W=0.9962),Translation=(X=-2,Y=-1,Z=1),Scale3D=(X=1,Y=1,Z=1))))
    WeightMinimum=0
    Weight=RigVMVariableNode() {variable=L clavicle up}.Value
   )
   n9 = RigVMVariableNode() {variable=Head Left}
   > &RigUnit_ModifyTransforms(
    ItemToModify="((Item=(Type=Bone,Name=\"dread_clavicle_r\"),Transform=(Rotation=(X=-0.13,Y=-0.0864,Z=-0.0114,W=0.9877),Translation=(X=-3,Y=-3,Z=-1),Scale3D=(X=1,Y=1,Z=1))),(Item=(Type=Bone,Name=\"dread_spine_r\"),Transform=(Rotation=(X=0,Y=0,Z=0,W=1),Translation=(X=0,Y=-2,Z= ...(truncated)"
    WeightMinimum=0
    Weight=@n9.Value
   )
   > &RigUnit_ModifyTransforms(
    ItemToModify=((Item=(Type=Bone,Name="earDreadlock_r_1"),Transform=(Rotation=(X=0,Y=0,Z=-0.1736,W=0.9848),Translation=(X=0,Y=-2,Z=0),Scale3D=(X=1,Y=1,Z=1))))
    WeightMinimum=0
    Weight=@n9.Value
   )
   > &RigUnit_ModifyTransforms(
    ItemToModify="((Item=(Type=Bone,Name=\"dread_clavicle_r\"),Transform=(Rotation=(X=-0.0789,Y=0.0941,Z=-0.0789,W=0.9893),Translation=(X=-2,Y=-1,Z=2),Scale3D=(X=1,Y=1,Z=1))),(Item=(Type=Bone,Name=\"dread_spine_r\"),Transform=(Rotation=(X=0,Y=0,Z=0,W=1),Translation=(X=0,Y=0,Z=0 ...(truncated)"
    WeightMinimum=0
    Weight=RigVMVariableNode() {variable=R clavicle up}.Value
   )
   > &RigUnit_ModifyTransforms(
    ItemToModify="((Item=(Type=Bone,Name=\"dread_clavicle_r\"),Transform=(Rotation=(X=0.0114,Y=-0.0864,Z=0.13,W=0.9877),Translation=(X=-2,Y=-1,Z=0),Scale3D=(X=1,Y=1,Z=1))),(Item=(Type=Bone,Name=\"dread_spine_r\"),Transform=(Rotation=(X=0,Y=0,Z=0,W=1),Translation=(X=0,Y=0,Z=2),S ...(truncated)"
    WeightMinimum=0
    Weight=RigVMVariableNode() {variable=R clavicle Back}.Value
   )
   n10 = RigVMVariableNode() {variable=Head Up}
   > &RigUnit_ModifyTransforms(
    ItemToModify="((Item=(Type=Bone,Name=\"dread_spine_r\"),Transform=(Rotation=(X=0.1605,Y=0.1354,Z=-0.2734,W=0.9387),Translation=(X=4,Y=0,Z=0),Scale3D=(X=1,Y=1,Z=1))),(Item=(Type=Bone,Name=\"dread_spine_l\"),Transform=(Rotation=(X=0.0242,Y=0.1371,Z=-0.172,W=0.9752),Translatio ...(truncated)"
    WeightMinimum=0
    Weight=@n10.Value
   )
   > n11 = &RigUnit_ModifyTransforms(
    ItemToModify="((Item=(Type=Bone,Name=\"earDreadlock_r_1\"),Transform=(Rotation=(X=-0.0114,Y=0.13,Z=0.0864,W=0.9877),Translation=(X=0,Y=0,Z=0),Scale3D=(X=1,Y=1,Z=1))),(Item=(Type=Bone,Name=\"earDreadlock_r_2\"),Transform=(Rotation=(X=0,Y=0,Z=0,W=1),Translation=(X=-4,Y=0,Z=0) ...(truncated)"
    WeightMinimum=0
    Weight=@n10.Value
   )
   n12 = &RigVMFunction_ControlFlowBranch(ExecuteContext=@n11.ExecutePin, Condition=RigVMVariableNode() {variable=RB Enable}.Value)
   &RigUnit_ModifyTransforms(
    ItemToModify="((Item=(Type=Bone,Name=\"asw_dreadFace1\"),Transform=(Rotation=(X=0.2666,Y=0.103,Z=-0.1932,W=0.9386),Translation=(X=-1.2867,Y=-2.4137,Z=1.4161),Scale3D=(X=1,Y=1,Z=1))),(Item=(Type=Bone,Name=\"asw_dreadFace2\"),Transform=(Rotation=(X=-0.0352,Y=-0.1923,Z=-0.1768 ...(truncated)"
    WeightMinimum=0
    ExecutePin=@n12.False
    Weight=RigVMVariableNode() {variable=Head Up}.Value
   )
   > &RigUnit_ModifyTransforms(
    ItemToModify="((Item=(Type=Bone,Name=\"asw_dreadFace1\"),Transform=(Rotation=(X=0,Y=0,Z=0.1636,W=0.9865),Translation=(X=0,Y=0,Z=0),Scale3D=(X=1,Y=1,Z=1))),(Item=(Type=Bone,Name=\"asw_dreadFace2\"),Transform=(Rotation=(X=0.1016,Y=-0.1954,Z=0.2924,W=0.9306),Translation=(X=0,Y ...(truncated)"
    WeightMinimum=0
    Weight=RigVMVariableNode() {variable=Head Down}.Value
   )
   > &RigUnit_ModifyTransforms(
    ItemToModify="((Item=(Type=Bone,Name=\"asw_dreadFace1\"),Transform=(Rotation=(X=0,Y=-0.0532,Z=0,W=0.9986),Translation=(X=0,Y=0,Z=0),Scale3D=(X=1,Y=1,Z=1))),(Item=(Type=Bone,Name=\"asw_dreadFace2\"),Transform=(Rotation=(X=0,Y=-0.2566,Z=0,W=0.9665),Translation=(X=0.294,Y=0,Z= ...(truncated)"
    WeightMinimum=0
    Weight=RigVMVariableNode() {variable=Head Right}.Value
   )
   n13 = &RigVMFunction_ControlFlowBranch(ExecuteContext=@RigVMFunction_Sequence.D, Condition=RigVMVariableNode() {variable=RB Enable}.Value)
   n14 = &RigVMFunction_MathBoolToFloat(Value=RigVMVariableNode() {variable=Jump}.Value)
   n15 = RigVMVariableNode() {variable=SkyDive}
   &RigUnit_ModifyTransforms(
    ItemToModify="((Item=(Type=Bone,Name=\"dread_clavicle_r\"),Transform=(Rotation=(X=-0.0151,Y=0.0858,Z=0.173,W=0.9811),Translation=(X=-1,Y=-2,Z=1),Scale3D=(X=1,Y=1,Z=1))),(Item=(Type=Bone,Name=\"dread_spine_r\"),Transform=(Rotation=(X=0,Y=0.061,Z=0,W=0.9981),Translation=(X=0, ...(truncated)"
    WeightMinimum=0
    Weight=&RigVMFunction_KalmanFloat(
     BufferSize=12
     Value=RigVMDispatchNode(True=0, False=@n14.Result, Condition=@n15.Value) {template="DISPATCH_RigVMDispatch_If(in Condition,in True,in False,out Result)"}.Result
    ) {template="KalmanFilter::Execute(in Value,in BufferSize,out Result)"}.Result
    ExecutePin=@n13.True
   )
   n16 = &RigVMFunction_MathBoolToFloat(Value=RigVMVariableNode() {variable=falling}.Value)
   n17 = RigVMVariableNode() {variable=SkyDive}
   > RigVMVariableNode(Value=&RigVMFunction_KalmanFloat(
     BufferSize=20
     Value=RigVMDispatchNode(True=0, False=@n16.Result, Condition=@n17.Value) {template="DISPATCH_RigVMDispatch_If(in Condition,in True,in False,out Result)"}.Result
    ) {template="KalmanFilter::Execute(in Value,in BufferSize,out Result)"}.Result) {variable=evaluate}
   n18 = RigVMVariableNode() {variable=evaluate}
   DISPATCH_RigVMDispatch_If = RigVMDispatchNode(
    False=&RigVMFunction_MathDoubleRemap(SourceMaximum=1, SourceMinimum=0.7, TargetMaximum=1, TargetMinimum=0, Value=@n18.Value).Result
    True=@n18.Value
    Condition=RigVMVariableNode() {variable=falling}.Value
   ) {template="DISPATCH_RigVMDispatch_If(in Condition,in True,in False,out Result)"}
   > &RigUnit_ModifyTransforms(
    ItemToModify="((Item=(Type=Bone,Name=\"dread_clavicle_r\"),Transform=(Rotation=(X=0,Y=-0.061,Z=0,W=0.9981),Translation=(X=-1,Y=-1,Z=0),Scale3D=(X=1,Y=1,Z=1))),(Item=(Type=Bone,Name=\"dread_spine_r\"),Transform=(Rotation=(X=0,Y=-0.061,Z=0,W=0.9981),Translation=(X=0.5,Y=0,Z=2 ...(truncated)"
    WeightMinimum=0
    Weight=@DISPATCH_RigVMDispatch_If.Result
   )
   RigVMFunction_MathBoolToFloat = &RigVMFunction_MathBoolToFloat(Value=RigVMVariableNode() {variable=Targeting}.Value)
   n19 = &RigVMFunction_KalmanFloat(BufferSize=6, Value=@RigVMFunction_MathBoolToFloat.Result) {template="KalmanFilter::Execute(in Value,in BufferSize,out Result)"}
   &RigUnit_ParentConstraint(
    AdvancedSettings=(InterpolationType=Average,RotationOrderForFilter=XZY)
    Child=(Type=Bone,Name="earDreadlock_r_2")
    Parents=((Item=(Type=Bone,Name="neck_02"),Weight=1))
    Weight=@n19.Result
    ExecutePin=@RigVMFunction_Sequence.C
   )
   > &RigUnit_ParentConstraint(
    AdvancedSettings=(InterpolationType=Average,RotationOrderForFilter=XZY)
    Child=(Type=Bone,Name="earDreadlock_r_B1")
    Parents=((Item=(Type=Bone,Name="clavicle_r"),Weight=1))
    Weight=1
   )
   > &RigUnit_ParentConstraint(
    AdvancedSettings=(InterpolationType=Average,RotationOrderForFilter=XZY)
    Child=(Type=Bone,Name="earDreadlock_r_A1")
    Parents=((Item=(Type=Bone,Name="clavicle_r"),Weight=1))
    Weight=1
   )
   > &RigUnit_ModifyTransforms(
    ItemToModify="((Item=(Type=Bone,Name=\"earDreadlock_r_1\"),Transform=(Rotation=(X=0.0019,Y=0.0829,Z=0.0229,W=0.9963),Translation=(X=0.3513,Y=0,Z=1.4549),Scale3D=(X=1,Y=1,Z=1))),(Item=(Type=Bone,Name=\"earDreadlock_r_2\"),Transform=(Rotation=(X=0,Y=-0.0205,Z=0,W=0.9998),Tran ...(truncated)"
    WeightMinimum=0
    Weight=@n19.Result
   )
   n20 = &RigVMFunction_MathQuaternionToEuler(
    RotationOrder=ZYX
    Value=&RigUnit_GetRelativeTransformForItem(Child=(Type=Bone,Name="head"), Parent=(Type=Bone,Name="spine_05"), bChildInitial=False, bParentInitial=False).RelativeTransform.Rotation
   )
   Multiply = &RigVMFunction_MathDoubleMul(
    A=@n19.Result
    B=&RigVMFunction_MathDoubleRemap(SourceMaximum=-40, SourceMinimum=5, TargetMaximum=1, TargetMinimum=0, bClamp=False, Value=@n20.Result.Z).Result
   )
   > &RigUnit_ModifyTransforms(
    ItemToModify="((Item=(Type=Bone,Name=\"earDreadlock_r_A1\"),Transform=(Rotation=(X=0,Y=0,Z=0.1305,W=0.9914),Translation=(X=0,Y=0,Z=0),Scale3D=(X=1,Y=1,Z=1))),(Item=(Type=Bone,Name=\"earDreadlock_r_B1\"),Transform=(Rotation=(X=0,Y=0,Z=0.1305,W=0.9914),Translation=(X=0,Y=0,Z= ...(truncated)"
    WeightMinimum=0
    Weight=@Multiply.Result
   )
   n21 = &RigVMFunction_MathDoubleRemap(SourceMaximum=25, SourceMinimum=-5, TargetMaximum=1, TargetMinimum=0, bClamp=False, Value=@n20.Result.Z)
   > &RigUnit_ModifyTransforms(
    ItemToModify="((Item=(Type=Bone,Name=\"dread_clavicle_r\"),Transform=(Rotation=(X=0,Y=0,Z=0,W=1),Translation=(X=0,Y=-2,Z=0),Scale3D=(X=1,Y=1,Z=1))),(Item=(Type=Bone,Name=\"dread_spine_r\"),Transform=(Rotation=(X=0,Y=0,Z=0,W=1),Translation=(X=0,Y=-2,Z=0),Scale3D=(X=1,Y=1,Z=1 ...(truncated)"
    WeightMinimum=0
    Weight=@n21.Result
   )
   Multiply_1 = RigVMAggregateNode(C=1, A=@n21.Result, B=RigVMVariableNode() {variable=R clavicle up}.Value)
   > n22 = &RigUnit_ModifyTransforms(
    ItemToModify=((Item=(Type=Bone,Name="dread_clavicle_r"),Transform=(Rotation=(X=0,Y=0,Z=0,W=1),Translation=(X=0,Y=-2,Z=0),Scale3D=(X=1,Y=1,Z=1))))
    WeightMinimum=0
    Weight=@Multiply_1.Result
   )
   RigVMFunction_ControlFlowBranch = &RigVMFunction_ControlFlowBranch(ExecuteContext=@n22.ExecutePin, Condition=RigVMVariableNode() {variable=RB Enable}.Value)
   &RigUnit_ModifyTransforms(
    ItemToModify="((Item=(Type=Bone,Name=\"asw_dreadFace1\"),Transform=(Rotation=(X=0.3065,Y=-0.1382,Z=-0.0164,W=0.9416),Translation=(X=0,Y=0,Z=0),Scale3D=(X=1,Y=1,Z=1))),(Item=(Type=Bone,Name=\"asw_dreadFace2\"),Transform=(Rotation=(X=0,Y=0,Z=0,W=1),Translation=(X=0,Y=0,Z=0),S ...(truncated)"
    WeightMinimum=0
    Weight=@n19.Result
    ExecutePin=@RigVMFunction_ControlFlowBranch.False
   )
   > &RigUnit_ModifyTransforms(
    ItemToModify="((Item=(Type=Bone,Name=\"asw_dreadFace1\"),Transform=(Rotation=(X=0,Y=0,Z=0,W=1),Translation=(X=0,Y=0,Z=0),Scale3D=(X=1,Y=1,Z=1))),(Item=(Type=Bone,Name=\"asw_dreadFace2\"),Transform=(Rotation=(X=0,Y=0,Z=0,W=1),Translation=(X=0,Y=0,Z=0),Scale3D=(X=1,Y=1,Z=1))) ...(truncated)"
    WeightMinimum=0
    Weight=&RigVMFunction_MathDoubleMul(A=@n19.Result, B=RigVMVariableNode() {variable=Head Left}.Value).Result
   )
   &RigUnit_ModifyTransforms(
    ExecutePin=()
    ItemToModify=((Item=(Type=Bone,Name="dread_clavicle_l"),Transform=(Rotation=(X=0,Y=0.1736,Z=0,W=0.9848),Translation=(X=0,Y=0,Z=-2),Scale3D=(X=1,Y=1,Z=1))))
    WeightMinimum=0
    Weight=&RigVMFunction_MathDoubleRemap(SourceMaximum=-2, SourceMinimum=-0.5, TargetMaximum=1, TargetMinimum=0, bClamp=False, Value=@n8.Value).Result
   )
  ]}
]
hierarchy [
 root {loc=<0 0 0>, rot=<0 0 0 1>}
 pelvis {parent=root, loc=<0 -0.5301 98.0571>, rot=<0.0127 0.707 -0.0127 -0.707>}
 spine_01 {parent=pelvis, loc=<6.1208 -0.9328 0>, rot=<0 0 0.034 -0.9994>}
 spine_02 {parent=spine_01, loc=<6.5996 -0.2095 0>, rot=<0 0 0.0238 -0.9997>}
 spine_03 {parent=spine_02, loc=<6.4914 -0.1237 0>, rot=<0 0 0.0149 -0.9999>}
 spine_04 {parent=spine_03, loc=<6.9598 0.0794 0>, rot=<0 0 -0.0132 -0.9999>}
 spine_05 {parent=spine_04, loc=<6.9987 0 0>, rot=<0 0 -0.0015 -1>}
 neck_01 {parent=spine_05, loc=<11.3138 0.1529 0>, rot=<0 0 -0.2003 -0.9797>}
 neck_02 {parent=neck_01, loc=<4.5627 0 0>, rot=<0 0 0.0504 -0.9987>}
 head {parent=neck_02, loc=<5.3914 0 0>, rot=<0 0 0.115 -0.9934>}
 FaceDreadlockFront_r_1 {parent=head, loc=<9.3444 7.1844 4.8013>, rot=<-0.0161 -0.9538 0.0075 0.2998>}
 FaceDreadlockFront_r_2 {parent=FaceDreadlockFront_r_1, loc=<5.2028 0 0>, rot=<0.0112 0.1216 -0.0498 -0.9913>}
 FaceDreadlockFront_r_3 {parent=FaceDreadlockFront_r_2, loc=<4.4155 0 0>, rot=<0.0449 0.2996 -0.006 -0.953>}
 FaceDreadlockFront_r_4 {parent=FaceDreadlockFront_r_3, loc=<5.8795 0 0>, rot=<-0.0358 -0.1854 -0.0352 -0.9814>}
 FaceDreadlockFront_r_5 {parent=FaceDreadlockFront_r_4, loc=<5.2433 0 0>, rot=<-0.016 -0.1362 0.1088 -0.9846>}
 FaceDreadlockFront_r_6 {parent=FaceDreadlockFront_r_5, loc=<8.0009 0 0>, rot=<0.9957 -0.0032 0.0863 0.0328>}
 FaceDreadlockFront_l_1 {parent=head, loc=<5.6287 1.491 -8.0861>, rot=<0.0602 -0.9954 0.0039 -0.0738>}
 FaceDreadlockFront_l_2 {parent=FaceDreadlockFront_l_1, loc=<3.5878 0.0584 -0.0324>, rot=<0.0001 -0.1294 -0.1103 -0.9854>}
 FaceDreadlockFront_l_3 {parent=FaceDreadlockFront_l_2, loc=<5.8375 -0.2845 0.0181>, rot=<0.0065 0.061 0.0036 -0.9981>}
 FaceDreadlockFront_l_4 {parent=FaceDreadlockFront_l_3, loc=<5.8729 0.3115 0.1598>, rot=<-0.0061 -0.0453 -0.03 -0.9985>}
 FaceDreadlockFront_l_5 {parent=FaceDreadlockFront_l_4, loc=<5.057 0.4395 0.8842>, rot=<0.0286 0.1018 -0.1093 -0.9884>}
 FaceDreadlockSide_r_1 {parent=head, loc=<10.2382 5.4669 5.2193>, rot=<0.099 -0.9482 -0.0289 0.3006>}
 FaceDreadlockSide_r_2 {parent=FaceDreadlockSide_r_1, loc=<5.8496 0 0>, rot=<-0.0616 0.3891 -0.0393 -0.9183>}
 FaceDreadlockSide_r_3 {parent=FaceDreadlockSide_r_2, loc=<6.2205 0 0>, rot=<-0.003 -0.0785 -0.152 -0.9853>}
 FaceDreadlockSide_r_4 {parent=FaceDreadlockSide_r_3, loc=<6.507 0 0>, rot=<-0.0048 -0.0189 -0.0557 -0.9983>}
 FaceDreadlockSide_r_5 {parent=FaceDreadlockSide_r_4, loc=<7.5493 0 0>, rot=<0 0 0 1>}
 FACIAL_C_FacialRoot {parent=head, loc=<1.7032 0.1628 -0.0006>, rot=<0.4847 -0.5148 -0.4847 -0.5148>}
 FACIAL_C_Forehead {parent=FACIAL_C_FacialRoot, loc=<-0.0013 -3.0197 8.1096>, rot=<-0.0397 0 0.0001 -0.9992>}
 FACIAL_L_ForeheadIn {parent=FACIAL_C_FacialRoot, loc=<1.832 -3.0213 8.0614>, rot=<-0.0569 -0.0696 0.0046 -0.9959>}
 FACIAL_R_ForeheadIn {parent=FACIAL_C_FacialRoot, loc=<-1.8309 -3.0269 8.054>, rot=<-0.0569 0.0696 -0.0045 -0.9959>}
 FACIAL_L_ForeheadMid {parent=FACIAL_C_FacialRoot, loc=<3.3766 -3.3155 7.5516>, rot=<-0.067 -0.1757 -0.0199 -0.982>}
 FACIAL_L_ForeheadOut {parent=FACIAL_C_FacialRoot, loc=<4.732 -3.3528 6.6265>, rot=<-0.1167 -0.256 -0.0779 -0.9565>}
 FACIAL_R_ForeheadOut {parent=FACIAL_C_FacialRoot, loc=<-4.732 -3.3564 6.6213>, rot=<-0.1166 0.256 0.0779 -0.9564>}
 FACIAL_L_EyesackUpper {parent=FACIAL_C_FacialRoot, loc=<3.3514 -2.2441 8.0766>, rot=<0.0565 -0.0847 -0.0638 -0.9928>}
 FACIAL_R_EyesackUpper {parent=FACIAL_C_FacialRoot, loc=<-3.3552 -2.2428 8.074>, rot=<0.0566 0.0847 0.0639 -0.9927>}
 FACIAL_L_EyelidUpperFurrow {parent=FACIAL_C_FacialRoot, loc=<3.1415 -1.9661 7.876>, rot=<-0.0199 -0.1516 -0.0781 -0.9852>}
 FACIAL_R_EyelidUpperFurrow {parent=FACIAL_C_FacialRoot, loc=<-3.1455 -1.9654 7.8726>, rot=<-0.0198 0.1517 0.0781 -0.9851>}
 FACIAL_L_EyelidLowerA {parent=FACIAL_C_FacialRoot, loc=<2.9483 -1.0325 5.87>, rot=<0.2974 0 0 -0.9547>}
 FACIAL_R_EyelidLowerA {parent=FACIAL_C_FacialRoot, loc=<-2.9548 -1.0322 5.8768>, rot=<0.2976 0 0 -0.9547>}
 FACIAL_L_EyeCornerInner {parent=FACIAL_C_FacialRoot, loc=<1.6889 -0.3629 6.9457>, rot=<0.014 0 0 -0.9999>}
 FACIAL_R_EyeCornerInner {parent=FACIAL_C_FacialRoot, loc=<-1.6885 -0.3695 6.949>, rot=<0.0142 0 0 -0.9999>}
 FACIAL_L_EyeCornerOuter {parent=FACIAL_C_FacialRoot, loc=<4.4379 -1.0231 6.5114>, rot=<0.0129 -0.3825 0.0053 -0.9238>}
 FACIAL_R_EyeCornerOuter {parent=FACIAL_C_FacialRoot, loc=<-4.4418 -1.0233 6.5146>, rot=<0.0131 0.3826 -0.0055 -0.9238>}
 FACIAL_L_EyesackLower {parent=FACIAL_C_FacialRoot, loc=<3.4479 0.5161 6.0595>, rot=<-0.0084 -0.14 -0.0203 -0.9899>}
 FACIAL_R_EyesackLower {parent=FACIAL_C_FacialRoot, loc=<-3.4364 0.5164 6.075>, rot=<-0.0082 0.14 0.0202 -0.9899>}
 FACIAL_L_CheekInner {parent=FACIAL_C_FacialRoot, loc=<2.1375 1.1197 6.8555>, rot=<-0.1129 -0.0901 -0.1962 -0.9699>}
 FACIAL_R_CheekInner {parent=FACIAL_C_FacialRoot, loc=<-2.1407 1.1208 6.8615>, rot=<-0.1129 0.0901 0.1962 -0.9699>}
 FACIAL_L_CheekOuter {parent=FACIAL_C_FacialRoot, loc=<4.1089 1.0468 5.4534>, rot=<0.0669 -0.3756 0.1051 -0.9184>}
 FACIAL_R_CheekOuter {parent=FACIAL_C_FacialRoot, loc=<-4.1131 1.0461 5.4544>, rot=<0.0669 0.3757 -0.1052 -0.9183>}
 FACIAL_C_Nose {parent=FACIAL_C_FacialRoot, loc=<-0.0011 0.3755 8.202>, rot=<-0.1244 0 0 -0.9922>}
 FACIAL_C_NoseLower {parent=FACIAL_C_Nose, loc=<0.0002 3.1269 -0.3123>, rot=<0.1304 0 0 -0.9915>}
 FACIAL_L_Nostril {parent=FACIAL_C_Nose, loc=<0.9771 2.439 -0.0619>, rot=<0.0173 0.1305 0.0023 -0.9913>}
 FACIAL_R_Nostril {parent=FACIAL_C_Nose, loc=<-0.9771 2.4393 -0.0617>, rot=<0.0173 -0.1305 -0.0023 -0.9913>}
 FACIAL_C_NoseTip {parent=FACIAL_C_Nose, loc=<0.0003 2.4425 0.5992>, rot=<0.0872 0 0 -0.9962>}
 FACIAL_C_MouthUpper {parent=FACIAL_C_FacialRoot, loc=<-0.0009 4.6109 6.3768>, rot=<0.0053 0 0 -1>}
 FACIAL_C_LipUpper {parent=FACIAL_C_MouthUpper, loc=<0 0.0665 2.3554>, rot=<-0.0872 0 0 -0.9962>}
 FACIAL_L_Masseter {parent=FACIAL_C_FacialRoot, loc=<6.3777 1.627 3.0362>, rot=<0.0751 0.1129 0.0072 -0.9907>}
 FACIAL_R_Masseter {parent=FACIAL_C_FacialRoot, loc=<-6.3796 1.6263 3.0356>, rot=<0.0751 -0.1129 -0.0072 -0.9907>}
 FACIAL_C_NoseBridge {parent=FACIAL_C_FacialRoot, loc=<-0.0011 -1.7264 8.8942>, rot=<0.0487 0 0 -0.9988>}
 FACIAL_L_EyelidLowerB {parent=FACIAL_C_FacialRoot, loc=<2.9483 -1.0325 5.87>, rot=<0.3305 0 0 -0.9438>}
 FACIAL_R_EyelidLowerB {parent=FACIAL_C_FacialRoot, loc=<-2.9548 -1.0322 5.8768>, rot=<0.339 0 0 -0.9408>}
 deform_megaLookUp {parent=head, loc=<14.7325 0.429 0>, rot=<0.0029 -0.7071 -0.0029 -0.7071>}
 back_dread_base {parent=deform_megaLookUp, loc=<0 -11.9306 -9.6122>, rot=<0 0 0 1>}
 dread_spine_r {parent=back_dread_base, loc=<-3.5715 2.7335 -16.4599>, rot=<-0.3526 -0.3489 0.5951 -0.6323>}
 dyn_dread_spine_r_A {parent=dread_spine_r, loc=<6.6294 0.779 -0.7619>, rot=<0.039 -0.1513 0.0237 -0.9874>}
 dyn_dread_spine_r_B {parent=dread_spine_r, loc=<7.4908 -3.9758 -0.6848>, rot=<0.0474 -0.1325 0.2448 -0.9593>}
 dread_spine_l {parent=back_dread_base, loc=<3.5715 2.7335 -16.4595>, rot=<-0.3292 -0.3709 0.5521 -0.6702>}
 dyn_dread_spine_l {parent=dread_spine_l, loc=<5.6453 -0.3024 -0.4283>, rot=<0 -0.0035 0 -1>}
 nape {parent=back_dread_base, loc=<0 -1.1553 -4.8492>, rot=<0 0 0 1>}
 earDreadlock_r_1 {parent=deform_megaLookUp, loc=<-9.7885 -0.3191 -12.1101>, rot=<-0.0016 -0.5848 -0.0023 -0.8112>}
 earDreadlock_r_2 {parent=earDreadlock_r_1, loc=<6.4166 0 0>, rot=<0.0063 -0.1239 0.0558 -0.9907>}
 earDreadlock_r_B1 {parent=earDreadlock_r_2, loc=<5.2047 -0.3286 0.3154>, rot=<-0.0607 -0.221 -0.1253 -0.9653>}
 earDreadlock_r_A1 {parent=earDreadlock_r_2, loc=<4.7452 2.0393 0.37>, rot=<0.0194 0.0255 -0.2341 -0.9717>}
 dread_clavicle_l {parent=deform_megaLookUp, loc=<7.6854 -6.3491 -23.9545>, rot=<-0.1854 -0.3023 0.4654 -0.811>}
 dyn_dread_clavicle_l {parent=dread_clavicle_l, loc=<8.8373 -0.0317 -0.8824>, rot=<-0.0132 -0.23 -0.0559 -0.9715>}
 dread_clavicle_r {parent=deform_megaLookUp, loc=<-7.883 -5.8965 -23.6715>, rot=<-0.226 -0.2766 0.7465 -0.5614>}
 dyn_dread_clavicle_r {parent=dread_clavicle_r, loc=<6.0826 -0.6936 0.3587>, rot=<0.0684 -0.1392 0.0311 -0.9874>}
 sheafBase {parent=deform_megaLookUp, loc=<0 -8.5577 -3.788>, rot=<-0.3446 0.6175 0.3446 -0.6175>}
 sheafCenter {parent=sheafBase, loc=<4.8239 0 0>, rot=<0 0 0 1>}
 asw_dreadFace1 {parent=head, loc=<5.0795 7.3674 7.7755>, rot=<-0.0863 -0.9959 0.0025 0.0259>}
 asw_dreadFace2 {parent=asw_dreadFace1, loc=<15.0442 0 0>, rot=<0 0 0 1>}
 clavicle_l {parent=spine_05, loc=<5.4831 -0.1053 -3.2946>, rot=<0.0465 -0.7289 -0.009 -0.6829>}
 upperarm_l {parent=clavicle_l, loc=<9.8658 -1.2136 -0.4105>, rot=<0.0164 -0.3245 -0.0269 -0.9454>}
 clavicle_r {parent=spine_05, loc=<5.483 -0.1053 3.2946>, rot=<0.7289 0.0465 -0.6829 0.009>}
 upperarm_r {parent=clavicle_r, loc=<-9.8658 1.2136 0.4105>, rot=<0.0164 -0.3245 -0.0269 -0.9454>}
 lowerarm_r {parent=upperarm_r, loc=<-26.465 0.8345 -0.259>, rot=<0.0036 0.0161 -0.2298 -0.9731>}
 hand_r {parent=lowerarm_r, loc=<-22.7355 0.3202 0.6106>, rot=<-0.6784 -0.0223 0.0735 -0.7307>}
 weapon_r {parent=hand_r, loc=<0.1974 3.411 -0.3811>, rot=<0 0 0 1>}
]
variables [
 Head Down {type=float, default=0}
 Head Left {type=float, default=0}
 Head Right {type=float, default=0}
 Head Roll Left {type=float, default=0}
 Head Roll Right {type=float, default=0}
 Head Up {type=float, default=0}
 Jump {type=bool, default=False}
 L clavicle Back {type=float, default=0}
 L clavicle up {type=float, default=0}
 R clavicle Back {type=float, default=0}
 R clavicle up {type=float, default=0}
 RB Enable {type=bool, default=False}
 SkyDive {type=bool, default=False}
 Targeting {type=bool, default=False}
 evaluate {type=float, default=0}
 falling {type=bool, default=False}
]
