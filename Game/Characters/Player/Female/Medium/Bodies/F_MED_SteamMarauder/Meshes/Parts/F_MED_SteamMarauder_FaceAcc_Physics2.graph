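# Physics asset for the face accessory mesh of the female medium SteamMarauder player body.
# It holds two sphere bodies, a larger offset one on sheafBase and a smaller one on deform_megaLookUp.
# A single constraint joins the two bones with narrow swing limits and an even tighter twist limit.
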PhysicsAsset /Game/Characters/Player/Female/Medium/Bodies/F_MED_SteamMarauder/Meshes/Parts/F_MED_SteamMarauder_FaceAcc_Physics2
bodies {
 deform_megaLookUp [{shapeType=Sphere, radius=3.1109, shapeName=deform_megaLookUp_sphere}]
 sheafBase [{shapeType=Sphere, radius=6.5185, shapeName=sheafBase_sphere, center=<4.1445 0 0>}]
}
constraints [{bone1Name=sheafBase, bone2Name=deform_megaLookUp, swing1LimitDegrees=20, swing2LimitDegrees=20, twistLimitDegrees=2}]
