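# Post-process style secondary-motion AnimBP for the SteamMarauder face accessory, animating dreads and nape on the female medium character.
# Copies the parent pose with the Body's right arm layered in, then applies head-space rigid-body physics and a state-driven Control Rig.
# Dread AnimDynamics blend over that cached pose, weakened while targeting.
# Output switches on targeting between spine-space RigidBody setups, with nape AnimDynamics only when not targeting.
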AnimBlueprint /BRCosmetics/Characters/Player/Female/Medium/Bodies/F_MED_SteamMarauder/Meshes/Parts/F_MED_SteamMarauder_FaceAcc_AnimBP
target_skeleton=/Game/Characters/Player/Female/Medium/Bodies/F_MED_SteamMarauder/Meshes/Parts/F_MED_SteamMarauder_FaceAcc_Skeleton
graphs [
 AnimGraph {nodes [
   n1 = RigidBody(ComponentPose=Local To Component(LocalPose=Use cached pose 'CachCR'())) {
    bClampLinearTranslationLimitToRefPose=true
    overridePhysicsAsset=/Game/Characters/Player/Female/Medium/Bodies/F_MED_SteamMarauder/Meshes/Parts/F_MED_SteamMarauder_FaceAcc_Physics
    simulationSpace=BaseBoneSpace
    alphaBoolBlend {blendOption=Linear}
    baseBoneRef {boneName=spine_05}
    componentAppliedLinearAccClamp=<1000 1000 1000>
    simSpaceSettings {maxAngularAcceleration=50, maxAngularVelocity=50, maxLinearAcceleration=30, maxLinearVelocity=50, velocityScaleZ=0, worldAlpha=0.584}
   }
   n2 = Anim Dynamics - Bone: nape(ComponentPose=Local To Component(LocalPose=Use cached pose 'CachCR'()), Alpha=0.7, ExternalForce="0,0,0") {
    angularBiasOverride=1
    angularDampingOverride=1
    angularSpringConstant=20
    bAngularSpring=true
    bGravityOverrideInSimSpace=true
    bLinearSpring=true
    bOverrideAngularBias=true
    bOverrideAngularDamping=true
    bOverrideLinearDamping=true
    bUsePlanarLimit=false
    bUseSphericalLimits=true
    gravityScale=0
    linearDampingOverride=0.997
    linearSpringConstant=5
    alphaBoolBlend {blendOption=Linear}
    boundBone {boneName=nape}
    componentAppliedLinearAccClamp=<3000 3000 3000>
    componentLinearAccScale=<0.05 0.05 0.05>
    componentLinearVelScale=<0.05 0.05 0.05>
    physicsBodyDefinitions [{
      collisionType=OuterSphere
      boundBone {boneName=nape}
      boxExtents=<5 5 5>
      constraintSetup {coneAngle=10, angularLimitsMax=<10 0 0>, angularLimitsMin=<-10 0 0>, angularTarget=<0 0 0>, linearAxesMax=<5 5 2>, linearAxesMin=<-5 -5 -10>}
      localJointOffset=<0 0 0>
     }]
    relativeSpaceBone {boneName=root}
    sphericalLimits [{limitRadius=12, drivingBone {boneName=head}, sphereLocalOffset=<2 0 0>}]
   }
   n3 = RigidBody(ComponentPose=@n2) {
    bClampLinearTranslationLimitToRefPose=true
    overridePhysicsAsset=&/F_MED_SteamMarauder_FaceAcc_Physics1
    simulationSpace=BaseBoneSpace
    alphaBoolBlend {blendOption=Linear}
    baseBoneRef {boneName=spine_05}
    componentAppliedLinearAccClamp=<1000 1000 1000>
    simSpaceSettings {maxAngularAcceleration=200, maxAngularVelocity=200, maxLinearAcceleration=200, maxLinearVelocity=200, velocityScaleZ=0, worldAlpha=0.5}
   }
   Output Pose(Result=Blend Poses by bool(
     BlendPose_0=Component To Local(ComponentPose=@n1)
     BlendPose_1=Component To Local(ComponentPose=@n3)
     BlendTime_0=0.2
     BlendTime_1=0.2
     bActiveValue=Get bIsTargeting()
    ) {blendType=Cubic}) {layerGroup=None}
   n4 = Anim Dynamics - Bone: dread_spine_r(ComponentPose=Local To Component(LocalPose=Use cached pose 'BaseRig'()), Alpha=0.5, ExternalForce="0,0,0") {
    angularBiasOverride=1
    angularDampingOverride=1
    angularSpringConstant=100
    bLinearSpring=true
    bOverrideLinearDamping=true
    bUseSphericalLimits=true
    gravityScale=0
    linearDampingOverride=0.995
    linearSpringConstant=5
    simulationSpace=RootRelative
    alphaBoolBlend {blendOption=Linear}
    boundBone {boneName=dread_spine_r}
    componentAppliedLinearAccClamp=<3000 3000 3000>
    componentLinearAccScale=<0.1 0.1 0.1>
    componentLinearVelScale=<0.1 0.1 0.1>
    physicsBodyDefinitions [{
      collisionType=OuterSphere
      boundBone {boneName=dread_spine_r}
      boxExtents=<5 5 5>
      constraintSetup {
       coneAngle=20
       angularLimitsMax=<10 10 10>
       angularLimitsMin=<-10 -10 -10>
       angularTarget=<0 0 0>
       linearAxesMax=<0.1 1 1>
       linearAxesMin=<-0.1 -1 -1>
      }
      localJointOffset=<0 0 0>
     }]
    planarLimits [{drivingBone {boneName=spine_05}, planeTransform {location=<0 -5 0>, rotation=<rot 0 10 -90>}}]
    relativeSpaceBone {boneName=head}
   }
   n5 = Anim Dynamics - Bone: dread_spine_l(ComponentPose=@n4, Alpha=0.5, ExternalForce="0,0,0") {
    angularBiasOverride=1
    angularDampingOverride=1
    angularSpringConstant=100
    bLinearSpring=true
    bOverrideLinearDamping=true
    bUseSphericalLimits=true
    gravityScale=0
    linearDampingOverride=0.9
    linearSpringConstant=5
    simulationSpace=RootRelative
    alphaBoolBlend {blendOption=Linear}
    boundBone {boneName=dread_spine_l}
    componentAppliedLinearAccClamp=<3000 3000 3000>
    componentLinearAccScale=<0.1 0.1 0.1>
    componentLinearVelScale=<0.1 0.1 0.1>
    physicsBodyDefinitions [{
      collisionType=OuterSphere
      boundBone {boneName=dread_spine_l}
      boxExtents=<5 5 5>
      constraintSetup {
       coneAngle=20
       angularLimitsMax=<10 10 10>
       angularLimitsMin=<-10 -10 -10>
       angularTarget=<0 0 0>
       linearAxesMax=<0.1 1 1>
       linearAxesMin=<-0.1 -1 -1>
      }
      localJointOffset=<0 0 0>
     }]
    planarLimits [{drivingBone {boneName=spine_05}, planeTransform {location=<0 -5 0>, rotation=<rot 0 9.9999 -90>}}]
    relativeSpaceBone {boneName=head}
   }
   n6 = Anim Dynamics - Bone: dread_clavicle_r(ComponentPose=@n5, Alpha=0.1, ExternalForce="0,0,0") {
    angularBiasOverride=1
    angularDampingOverride=1
    angularSpringConstant=5
    bAngularSpring=true
    bLinearSpring=true
    bOverrideAngularBias=true
    bOverrideLinearDamping=true
    bUseSphericalLimits=true
    gravityScale=0
    linearDampingOverride=0.9
    linearSpringConstant=5
    simulationSpace=RootRelative
    alphaBoolBlend {blendOption=Linear}
    boundBone {boneName=dread_clavicle_r}
    componentAppliedLinearAccClamp=<150 150 150>
    physicsBodyDefinitions [{
      collisionType=OuterSphere
      boundBone {boneName=dread_clavicle_r}
      boxExtents=<5 5 5>
      constraintSetup {
       coneAngle=5
       angularLimitsMax=<10 10 10>
       angularLimitsMin=<-10 -10 -10>
       angularTarget=<0 0 0>
       linearAxesMax=<0.1 0.5 0>
       linearAxesMin=<-0.1 -0.5 0>
      }
      localJointOffset=<10 0 0>
     }]
   }
   n7 = Anim Dynamics - Bone: dread_clavicle_l(ComponentPose=@n6, Alpha=0.1, ExternalForce="0,0,0") {
    angularBiasOverride=1
    angularDampingOverride=1
    angularSpringConstant=5
    bAngularSpring=true
    bLinearSpring=true
    bOverrideAngularBias=true
    bOverrideLinearDamping=true
    bUseSphericalLimits=true
    gravityScale=0
    linearDampingOverride=0.9
    linearSpringConstant=5
    simulationSpace=RootRelative
    alphaBoolBlend {blendOption=Linear}
    boundBone {boneName=dread_clavicle_l}
    componentAppliedLinearAccClamp=<150 150 150>
    physicsBodyDefinitions [{
      collisionType=OuterSphere
      boundBone {boneName=dread_clavicle_l}
      boxExtents=<5 5 5>
      constraintSetup {
       coneAngle=5
       angularLimitsMax=<10 10 10>
       angularLimitsMin=<-10 -10 -10>
       angularTarget=<0 0 0>
       linearAxesMax=<0.1 0.5 0>
       linearAxesMin=<-0.1 -0.5 0>
      }
      localJointOffset=<10 0 0>
     }]
   }
   Save cached pose 'CachCR'(Pose=Two Way Blend(A=Use cached pose 'BaseRig'(), B=Component To Local(ComponentPose=@n7), Alpha=Get AlphaLitleDyn() -> (AlphaLitleDyn:float)) {blendNode {alphaBoolBlend {blendOption=Linear}}}) {cacheName=CachCR}
   n8 = Layered blend per bone(
    BasePose=Copy Pose From Mesh() {bUseAttachedParent=true}
    BlendPoses_0=Copy Pose From Mesh(SourceMeshComponent=Get Body() -> (Body:Skeletal Mesh Component Object Reference))
   ) {layerSetup [{branchFilters [{boneName=clavicle_r}, {boneName=upperarm_r}, {boneName=lowerarm_r}, {boneName=weapon_r}]}]}
   n9 = RigidBody(ComponentPose=Local To Component(LocalPose=@n8), Alpha=0.8) {
    bClampLinearTranslationLimitToRefPose=true
    overridePhysicsAsset=/Game/Characters/Player/Female/Medium/Bodies/F_MED_SteamMarauder/Meshes/Parts/F_MED_SteamMarauder_FaceAcc_Physics2
    simulationSpace=BaseBoneSpace
    alphaBoolBlend {blendOption=Linear}
    baseBoneRef {boneName=head}
    componentAppliedLinearAccClamp=<100 100 100>
    simSpaceSettings {maxAngularAcceleration=50, maxAngularVelocity=50, maxLinearAcceleration=30, maxLinearVelocity=50, velocityScaleZ=0, worldAlpha=1}
   }
   n10 = Control Rig!(
    Source=Component To Local(ComponentPose=@n9)
    Alpha
    Targeting:bool=Get bIsTargeting()
    Jump:bool=Get bIsJumping()
    falling:bool=Get bIsFalling()
    SkyDive:bool=Get bIsSkyDiving()
    RB Enable:bool=Get bRigidBodyEnabled()
   ) -> (Pose) {
    alphaBoolBlend {blendOption=Linear}
    controlRigAssetReference {blueprintRigClass=/Game/Characters/Player/Female/Medium/Bodies/F_MED_SteamMarauder/Meshes/Parts/F_MED_SteamMarauder_FaceAcc_CtrlRig.F_MED_SteamMarauder_FaceAcc_CtrlRig_C}
   }
   Save cached pose 'BaseRig'(Pose=@n10) {cacheName=BaseRig}
  ]}
 EventGraph {nodes [
   Event BlueprintUpdateAnimation()
   > Set AlphaLitleDyn(AlphaLitleDyn:float=SelectFloat(A=0.2, B=1, bPickA=Get bIsTargeting())) -> (Output_Get:float)
   n1 = Event OnCharacterCustomizationLoaded()
   > Set Body(Body:Skeletal Mesh Component Object Reference=Get Array Item with EFort Custom Part Type(Array=@n1.CustomMeshParts)) -> (Output_Get:Skeletal Mesh Component Object Reference)
  ]}
]
variables [
 AlphaLitleDyn {type=number, default=1}
 Body {type=SkeletalMeshComponent}
]
